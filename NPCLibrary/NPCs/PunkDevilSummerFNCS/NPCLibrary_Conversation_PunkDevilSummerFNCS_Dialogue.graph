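# Conversation database for the PunkDevilSummerFNCS NPC, providing its dialogue about FNCS competition.
# An entry point tag opens a greeting speech with an "Ask about..." choice, which leads to an FNCS topic or a Back option.
# The FNCS choice carries a CR25 quest conversation gameplay tag and chains four boastful speech lines before a ReturnToStart node loops back to the start.
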ConversationDatabase /NPCLibrary/NPCs/PunkDevilSummerFNCS/NPCLibrary_Conversation_PunkDevilSummerFNCS_Dialogue
entry_points [{tag=Conversation.Entry.PunkDevilSummerFNCS.Dialogue, destinations [#1]}]
nodes [
 ConversationEntryPointNode_0 {
  id=#1
  class=ConversationEntryPointNode
  outputs [#2]
  properties {entryTag {tagName=Conversation.Entry.PunkDevilSummerFNCS.Dialogue}}
 }
 Speech_C_0 {
  id=#2
  class=Speech_C
  outputs [#3, #4]
  properties {
   nodeName=""
   softTaskIcon=&T_UI_NPC_question
   generalConfig {defaultMessage=Ridin' the waves of good conversation.}
  }
  sub_nodes [ConversationChoiceNode_5 {class=ConversationChoiceNode, properties {defaultChoiceDisplayText=Ask about...}}]
 }
 Back_C_0 {
  id=#4
  class=Back_C
  sub_nodes [ConversationChoiceNode_7 {class=ConversationChoiceNode, properties {defaultChoiceDisplayText=BACK}}]
 }
 Speech_C_7 {
  id=#3
  class=Speech_C
  outputs [#5]
  properties {
   nodeName=""
   softTaskIcon=&T_UI_NPC_question
   generalConfig {defaultMessage=I see every competition as mine to win.}
  }
  sub_nodes [ConversationChoiceNode_15 {
    class=ConversationChoiceNode
    properties {defaultChoiceDisplayText=FNCS, choiceTags {gameplayTags [{tagName=Athena.Quests.CR25.Conversations.PunkDevilSummerFNCS.FNCS}]}}
   }]
 }
 Speech_C_15 {
  id=#5
  class=Speech_C
  outputs [#6]
  properties {nodeName="", softTaskIcon=&T_UI_NPC_question, generalConfig {defaultMessage=And that includes FNCS.}}
 }
 Speech_C_14 {
  id=#6
  class=Speech_C
  outputs [#7]
  properties {
   nodeName=""
   softTaskIcon=&T_UI_NPC_question
   generalConfig {defaultMessage="If you think you can keep up, bring it on."}
  }
 }
 Speech_C_1 {
  id=#7
  class=Speech_C
  outputs [#8]
  properties {
   nodeName=""
   softTaskIcon=&T_UI_NPC_question
   generalConfig {defaultMessage=It'd be nice to have some worthy competition out there to watch me score the crown.}
  }
 }
 ReturnToStart_C_23 {
  id=#8
  class=ReturnToStart_C
  properties {softTaskIcon=&T_UI_NPC_Back}
  sub_nodes [ConversationChoiceNode_32 {class=ConversationChoiceNode, properties {defaultChoiceDisplayText=BACK}}]
 }
]
